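; AI evaluator that decides when a Fortnite bot builds in combat.
; It starts only if the console variable is on, the building component allows it, blackboard alert level is Threatened, and combined wood, stone and metal reach a minimum.
; It continues while bot and target are not confirmed in the same building grid cell.
; When execution becomes pending, it writes the pawn's grid-snapped location to the blackboard as a pre-build move destination.
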(event EventOnExecutionStatusChanged
  (params (PreviousExecutionStatus :type EExecutionStatus%20Enum :default "ExecutionError") (NewExecutionStatus :type EExecutionStatus%20Enum :default "ExecutionError"))
  (bind _returnvalue (CallFunction|GetOwner))
  (switch Utilities|FlowControl|Switch|SwitchonEExecutionStatus NewExecutionStatus
    (:ExecutionPending
      ;[1 1 1] Find closest building grid cell center and setup BBKey to move to it before start bulding
      (Utilities|IsValid _returnvalue
        (:Is%20Valid
          (bind _asaicontroller (Utilities|Casting|CastToAIController (Components|GetOwner _returnvalue)))
          (Utilities|IsValid (AI|GetBlackboard _asaicontroller)
            (:Is%20Valid
              (Utilities|IsValid (Pawn|GetControlledPawn _asaicontroller)
                (:Is%20Valid
                  (AI|Components|Blackboard|SetValueasVector (AI|GetBlackboard _asaicontroller) (Variables|Default|GetGridCellCenterKeyName) (Fortnite|SnaptoFortBuildingGrid (Transformation|GetActorLocation (Pawn|GetControlledPawn _asaicontroller)))))))))))))

(fn EvaluateOngoingConditions
  (params (OwnerComp :type Behavior%20Tree%20Component%20Object%20Reference))
  (return (not (CallFunction|IsinSameBuildingGridThanTarget))))

(fn EvaluateStartingConditions
  (params (OwnerComp :type Behavior%20Tree%20Component%20Object%20Reference))
  ;[1 1 1] CVar Enable/Disable
  (bind _returnvalue (Development|GetConsoleVariableBoolValue (Variables|Default|GetCombatBuidlingEnabledCVar)))
  (if _returnvalue
    ;[1 1 1] Building Allowed
    (if (CallFunction|IsBuildingAllowed)
      ;[1 1 1] In Combat
      (if (CallFunction|IsinCombat (Variables|GetOwnerComp))
        ;[1 1 1] Materials check
        (if (CallFunction|HasEnoughMaterials)
          (return true)
          (else
            (return false)))
        (else
          (return false)))
      (else
        (return false)))
    (else
      (return false))))

(fn HasEnoughMaterials
  (bind _asfort_athena_aibot_controller (Utilities|Casting|CastToFortAthenaAIBotController (Components|GetOwner (CallFunction|GetOwner)))
    (:then
      (return (>= (Utilities|Operators|Add (Class|FortAthenaAIBotController|GetResourcesCount _asfort_athena_aibot_controller "Wood") (Class|FortAthenaAIBotController|GetResourcesCount _asfort_athena_aibot_controller "Stone") (Class|FortAthenaAIBotController|GetResourcesCount _asfort_athena_aibot_controller "Metal")) (ScalableFloat|GetValueatLevel (Variables|Default|GetMinMaterials)))))
    (:CastFailed
      (return false))))

(fn IsBuildingAllowed
  (bind _asfort_athena_aibot_controller (Utilities|Casting|CastToFortAthenaAIBotController (Components|GetOwner (CallFunction|GetOwner)))
    (:then
      (if (Utilities|IsValid (Actor|GetComponentbyClass _asfort_athena_aibot_controller "/Script/FortniteAI.FortAthenaAIBotBuildingComponent") :!class /Script/Engine.KismetSystemLibrary)
        (return (Class|FortAthenaAIBotBuildingComponent|IsBuildingAllowed (Actor|GetComponentbyClass _asfort_athena_aibot_controller "/Script/FortniteAI.FortAthenaAIBotBuildingComponent")))
        (else
          (return false))))
    (:CastFailed
      (return false))))

(fn IsInCombat
  (params (BehaviorTreeComponent :type Behavior%20Tree%20Component%20Object%20Reference))
  (switch Utilities|FlowControl|Switch|SwitchonEAlertLevel (Utilities|Enum|BytetoEnumEAlertLevel (Blackboard|GetEnum (Variables|Default|GetAlertLevel) BehaviorTreeComponent))
    (:Unaware
      _)
    (:Suspicious
      _)
    (:LostTarget
      _)
    (:Threatened
      (return true))
    (:Count
      _))
  (return false))

(fn IsInSameBuildingGridThanTarget
  (bind _returnvalue (CallFunction|GetOwner))
  (Utilities|IsValid _returnvalue
    (:Is%20Valid
      (bind _asfort_athena_aibot_controller (Utilities|Casting|CastToFortAthenaAIBotController (Components|GetOwner _returnvalue))
        (:then
          (if (and (.ReturnValue (Class|FortAthenaAIBotController|GetCurrentBuildingGridIndices _asfort_athena_aibot_controller)) (.ReturnValue (Class|FortAthenaAIBotController|GetCurrentTargetBuildingGridIndices _asfort_athena_aibot_controller)))
            (return (CallFunction|IsSameCellGrid (Class|FortAthenaAIBotController|GetCurrentBuildingGridIndices _asfort_athena_aibot_controller) (Class|FortAthenaAIBotController|GetCurrentTargetBuildingGridIndices _asfort_athena_aibot_controller)))
            (else
              (return false))))
        (:CastFailed
          (return false))))
    (:Is%20Not%20Valid
      (return false))))

(fn IsSameCellGrid
  (params (BuildingCellIndexA_X :type Integer) (BuildingCellIndexA_Y :type Integer) (BuildingCellIndexA_Z :type Integer) (BuildingCellIndexB_X :type Integer) (BuildingCellIndexB_Y :type Integer) (BuildingCellIndexB_Z :type Integer))
  (return (Math|Boolean|ANDBoolean (== BuildingCellIndexA_X BuildingCellIndexB_X) (== BuildingCellIndexA_Y BuildingCellIndexB_Y) (== BuildingCellIndexA_Z BuildingCellIndexB_Z))))
