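; AI evaluator for Phoebe bots that tests whether the controlled player pawn is in a vehicle.
; Its EvaluateStartingConditions override calls the parent, casts the behavior tree component's owner to Controller and its pawn to FortPlayerPawn, then returns IsinVehicle.
; A Sequence node's second branch returns false as the fallback when either cast fails.
; The parent call's result does not feed the returned value.
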(fn EvaluateStartingConditions
  (params (OwnerComp :type Behavior%20Tree%20Component%20Object%20Reference))
  (|CallParentFunction OwnerComp)
  (Utilities|FlowControl|Sequence
    (:then_0
      (bind _ascontroller (Utilities|Casting|CastToController (Components|GetOwner (Variables|GetOwnerComp))))
      (bind _asfort_player_pawn (Utilities|Casting|CastToFortPlayerPawn (Pawn|GetControlledPawn _ascontroller)))
      (return (Vehicle|IsinVehicle _asfort_player_pawn)))
    (:then_1
      (return false))))
